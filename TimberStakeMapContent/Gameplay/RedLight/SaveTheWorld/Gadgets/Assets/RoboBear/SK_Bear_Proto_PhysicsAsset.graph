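# Physics asset for SK_Bear_Proto, the prototype RoboBear skeletal mesh among the Save the World gadget assets.
# Six capsule bodies cover the rig: tiny ones on the Root, upper and guns bones, and large ones on three polySurface bones that provide the actual volume.
# Constraints follow a Root, upper, guns chain, each with one polySurface body attached, all sharing identical swing and twist limit angles.
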PhysicsAsset /TimberStakeMapContent/Gameplay/RedLight/SaveTheWorld/Gadgets/Assets/RoboBear/SK_Bear_Proto_PhysicsAsset
bodies {
 Root [{shapeType=Capsule, length=0.505, radius=0.505, rotation=<rot 0 -90 0>}]
 guns [{shapeType=Capsule, length=0.505, radius=0.505, rotation=<rot 0 90 180>}]
 polySurface32 [{shapeType=Capsule, length=96.539, radius=45.9995, center=<0.3011 -2.1795 58.0002>, rotation=<rot 0 -89.932 -88.6521>}]
 polySurface67 [{shapeType=Capsule, length=161.8227, radius=38.3514, center=<0.0165 29.4197 164.8481>, rotation=<rot 0 -90.0203 -89.9931>}]
 polySurface68 [{shapeType=Capsule, length=109.5778, radius=68.4149, center=<-0.0553 2.1517 176.4519>, rotation=<rot 0 -179.2688 -52.7856>}]
 upper [{shapeType=Capsule, length=0.505, radius=0.505, rotation=<rot 0 -90 -90>}]
}
constraints [
 {bone1Name=upper, bone2Name=Root, swing1LimitDegrees=45, swing2LimitDegrees=45, twistLimitDegrees=45}
 {bone1Name=guns, bone2Name=upper, swing1LimitDegrees=45, swing2LimitDegrees=45, twistLimitDegrees=45}
 {bone1Name=polySurface67, bone2Name=guns, swing1LimitDegrees=45, swing2LimitDegrees=45, twistLimitDegrees=45}
 {bone1Name=polySurface68, bone2Name=upper, swing1LimitDegrees=45, swing2LimitDegrees=45, twistLimitDegrees=45}
 {bone1Name=polySurface32, bone2Name=Root, swing1LimitDegrees=45, swing2LimitDegrees=45, twistLimitDegrees=45}
]
